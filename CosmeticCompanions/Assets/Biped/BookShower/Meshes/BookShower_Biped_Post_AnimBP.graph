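# Post-process AnimBlueprint for the BookShower biped cosmetic companion, adding secondary motion to its input pose.
# A bool blend picks either the cached input or a version with pelvis constrained equally to both IK feet.
# That pose feeds chained spring-based AnimDynamics on two attach_emote bones and both eyes, producing the output; a PoseAsset node sits unconnected.
# The event graph derives EmoteOrAnimating from reverse retargeting or a companion/mannequin owner check.
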AnimBlueprint /CosmeticCompanions/Assets/Biped/BookShower/Meshes/BookShower_Biped_Post_AnimBP
target_skeleton=&/BookShower_Biped_Skeleton
graphs [
 AnimGraph {nodes [
   Pose!(PoseAsset:Pose Asset Object Reference=Get PoseAsset() -> (PoseAsset:Pose Asset Object Reference), SourcePose:Pose Link Structure) -> (Pose:Pose Link Structure) {is=title AnimGraphNode_PoseBlendNode, bIgnorePoseAssetSourceAnimationCurves=true, blendOption=Linear}
   Output Pose(Result=Use cached pose 'DynamicsCache'())
   Save cached pose 'CRCache_Input'(Pose=Input Pose()) {cacheName=CRCache_Input}
   n1 = Anim Dynamics - Bone: attach_emote_02(ComponentPose=Local To Component(LocalPose=Use cached pose 'preDynamicsCache'()), Alpha=1, ExternalForce="0,0,0") {
    angularDampingOverride=1
    angularSpringConstant=20000
    bAngularSpring=true
    bOverrideAngularDamping=true
    bUseGravityOverride=true
    gravityScale=0.35
    simulationSpace=RootRelative
    alphaScaleBiasClamp {bias=1, scale=-1}
    boundBone {boneName=attach_emote_02}
    componentAppliedLinearAccClamp=<500 500 500>
    componentLinearAccScale=<1 1 1>
    componentLinearVelScale=<1 1 1>
    gravityOverride {z=-980}
    physicsBodyDefinitions [{
      boundBone {boneName=attach_emote_02}
      boxExtents=<20 30 30>
      constraintSetup {angularLimitsMax=<5 30 15>, angularLimitsMin=<-5 -30 -5>, angularTarget=<1 0 0>}
      localJointOffset=<-2 0 0>
     }]
    simSpaceSettings {maxAngularAcceleration=12, maxAngularVelocity=3, simSpaceAngularAlpha=1}
   }
   n2 = Anim Dynamics - Bone: attach_emote_01(ComponentPose=@n1, Alpha=1, ExternalForce="0,0,0") {
    angularDampingOverride=0.9999
    angularSpringConstant=14000
    bAngularSpring=true
    bOverrideAngularDamping=true
    bUseGravityOverride=true
    gravityScale=0.35
    simulationSpace=RootRelative
    alphaScaleBiasClamp {bias=1, scale=-1}
    boundBone {boneName=attach_emote_01}
    componentAppliedLinearAccClamp=<500 500 500>
    componentLinearAccScale=<1 1 1>
    componentLinearVelScale=<1 1 1>
    gravityOverride {z=-980}
    physicsBodyDefinitions [{
      boundBone {boneName=attach_emote_01}
      boxExtents=<30 40 35>
      constraintSetup {angularLimitsMax=<10 10 5>, angularLimitsMin=<-10 -10 -5>, angularTarget=<1 0 0>}
      localJointOffset=<-2 0 0>
     }]
    simSpaceSettings {maxAngularAcceleration=12, maxAngularVelocity=4, simSpaceAngularAlpha=1}
   }
   n3 = Anim Dynamics - Bone: eye_l(ComponentPose=@n2, Alpha=1, ExternalForce="0,0,0") {
    angularDampingOverride=1
    angularSpringConstant=5000
    bAngularSpring=true
    bOverrideAngularDamping=true
    bUseGravityOverride=true
    gravityScale=0.35
    simulationSpace=RootRelative
    alphaBoolBlend {blendOption=Linear}
    alphaScaleBiasClamp {bias=1, scale=-1}
    boundBone {boneName=eye_l}
    componentAppliedLinearAccClamp=<500 500 500>
    componentLinearAccScale=<1 1 1>
    componentLinearVelScale=<1 1 1>
    gravityOverride {z=-980}
    physicsBodyDefinitions [{
      boundBone {boneName=eye_l}
      boxExtents=<15 15 15>
      constraintSetup {angularTargetAxis=AxisY, twistAxis=AxisY, angularLimitsMax=<20 5 11>, angularLimitsMin=<-10 -5 -5>, angularTarget=<0 1 0>}
      localJointOffset=<0 1.5 0>
     }]
    simSpaceSettings {maxAngularAcceleration=12, maxAngularVelocity=4, simSpaceAngularAlpha=1}
   }
   n4 = Anim Dynamics - Bone: eye_r(ComponentPose=@n3, Alpha=1, ExternalForce="0,0,0") {
    angularDampingOverride=1
    angularSpringConstant=5000
    bAngularSpring=true
    bOverrideAngularDamping=true
    bUseGravityOverride=true
    gravityScale=0.35
    simulationSpace=RootRelative
    alphaBoolBlend {blendOption=Linear}
    alphaScaleBiasClamp {bias=1, scale=-1}
    boundBone {boneName=eye_r}
    componentAppliedLinearAccClamp=<500 500 500>
    componentLinearAccScale=<1 1 1>
    componentLinearVelScale=<1 1 1>
    gravityOverride {z=-980}
    physicsBodyDefinitions [{
      boundBone {boneName=eye_r}
      boxExtents=<15 15 15>
      constraintSetup {angularTargetAxis=AxisY, twistAxis=AxisY, angularLimitsMax=<25 5 2.5>, angularLimitsMin=<-3 -5 -9>, angularTarget=<0 1 0>}
      localJointOffset=<0 -1.75 0>
     }]
    simSpaceSettings {maxAngularAcceleration=12, maxAngularVelocity=4, simSpaceAngularAlpha=1}
   }
   Save cached pose 'DynamicsCache'(Pose=Component To Local(ComponentPose=@n4)) {cacheName=DynamicsCache}
   n5 = "Constraint - pelvis (PP)"!(
    ComponentPose=Local To Component(LocalPose=Use cached pose 'CRCache_Input'())
    Alpha=1
    ConstraintWeights_0=0.5
    ConstraintWeights_1:float=0.5
   ) -> (Pose) {
    boneToModify {boneName=pelvis}
    constraintSetup [{targetBone {boneName=IK_Foot_L}}, {targetBone {boneName=IK_Foot_R}}]
    constraintWeights [1,1]
   }
   Save cached pose 'preDynamicsCache'(Pose=Blend Poses by bool(BlendPose_0=Use cached pose 'CRCache_Input'(), BlendPose_1=Component To Local(ComponentPose=@n5))) {cacheName=preDynamicsCache}
  ]}
 EventGraph {nodes [
   Event BlueprintInitializeAnimation()
   n1 = GetOwningActor()
   n2 = GetOwningActor()
   > Set isAnimatingGameplay(isAnimatingGameplay:bool=OR Boolean(
     A=IsA!(Object:Object Reference=@n2, ObjectClass:Object Class Reference=/CosmeticCompanions/Assets/Shared/Blueprint/BP_Companion.BP_Companion_C) -> (ReturnValue:bool)
     B=IsA(Object=@n1, ObjectClass=/Script/CosmeticCompanionsRuntime.FortCompanionMannequin)
    )) -> (Output_Get:bool)
   Event BlueprintUpdateAnimation()
   > Set EmoteOrAnimating(EmoteOrAnimating:bool=OR Boolean(A=Get bUseReverseRetarget() -> (bUseReverseRetarget:bool), B=Get isAnimatingGameplay() -> (isAnimatingGameplay:bool))) -> (Output_Get:bool)
  ]}
]
variables [
 EmoteOrAnimating {type=boolean}
 isAnimatingGameplay {type=boolean}
 PoseAsset {type=PoseAsset, default=/CosmeticCompanions/Automation_Template/companion_pose_asset_template}
]
